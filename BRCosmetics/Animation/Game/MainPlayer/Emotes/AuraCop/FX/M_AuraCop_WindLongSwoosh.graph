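# Translucent Niagara mesh-particle material for the long wind swoosh effect in the AuraCop emote.
# Dynamic parameters scale and offset the UVs of a clamped circular gradient texture, whose red channel is multiplied by particle alpha.
# XMult and XPan each drive both UV axes, so YMult and YPan are unused.
# No links to material outputs are shown, and a green-channel UV mask feeds nothing visible.
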Material /BRCosmetics/Animation/Game/MainPlayer/Emotes/AuraCop/FX/M_AuraCop_WindLongSwoosh
graph {
 properties {bAutomaticallySetUsageInEditor=false, bUsedWithNiagaraMeshParticles=true, blendMode=BLEND_Translucent, usedShadingModels=Default Lit}
 nodes [
  ComponentMask(None=TextureCoordinate()) {g=true}
  n1 = TextureCoordinate()
  n2 = DynamicParameter() {defaultValue=<rgba 0.7 0.1 0 0.5>, paramNames [XMult,YMult,XPan,YPan]}
  n3 = AppendVector(A=@n2.XMult, B=@n2.XMult)
  n4 = TextureSample(UVs=Add(A=Multiply(A=@n1, B=@n3), B=AppendVector(A=@n2.XPan, B=@n2.XPan))) {
   samplerSource=SSM_Clamp_WorldGroupSettings
   samplerType=SAMPLERTYPE_LinearColor
   texture=/Game/Global/Textures/Default/Masks_and_Gradients/LinearCircleGrad
  }
  Multiply(A=@n4.R, B=ParticleColor().A)
 ]
}
