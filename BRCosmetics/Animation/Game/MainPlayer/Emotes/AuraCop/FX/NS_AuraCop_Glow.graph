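# Niagara glow effect for the AuraCop emote, named HandGlow, that attaches glowing sprites with lights to the character's skeletal mesh.
# A burst spawns two long-lived particles, and each particle's spawn group doubles as its bone/socket index, pinning one glow to each point.
# Color pulses through blue-violet tints on the system's looped age, while a second scale pass fades brightness and alpha in and out.
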NiagaraSystem /BRCosmetics/Animation/Game/MainPlayer/Emotes/AuraCop/FX/NS_AuraCop_Glow
dependencies {
 usedDataInterfaces [&NiagaraDataInterfaceColorCurve, &NiagaraDataInterfaceCurve, &NiagaraDataInterfaceSkeletalMesh]
 usedDynamicInputs [&SimulationPosition]
 usedModules [
  &self:SystemSpawnScript.SystemScriptSource.SystemScriptGraph.NiagaraNodeAssignment_6.SetVariables_#1
  &SystemState
  &Timeline
 ]
}
emitters {HandGlow {
  emitterUpdateScript [
   &EmitterState(Life Cycle Mode=::NewEnumerator1, Loop Behavior=::NewEnumerator1, Loop Duration Mode=::NewEnumerator1, UseLoopDelay=false)
   &SpawnBurst_Instantaneous(Spawn Count=2, Use Spawn Probability=true)
  ]
  particleSpawnScript [
   &FN_SetSpawnGroup(Number of Spawn Groups=2)
   &InitializeParticle(
    Color=<rgba 1.2 0.4186 0 0.75>
    Lifetime=6
    Mass Mode=::NewEnumerator1
    Sprite Size Mode=::NewEnumerator3
    Uniform Sprite Size=55
    UsePositionOffset=false
    Write Material Random=true
   ) {module=InitializeParticle001}
  ]
  particleUpdateScript [
   &ParticleState()
   &SkeletalMeshLocation(
    Bone / Socket Index=$Particles.SpawnGroup
    Bone Sampling Mode=::NewEnumerator4
    Inherit Velocity=false
    Interpolate Mesh Sampling=false
    MaskSpawnGroup=false
    Sampled Position Offset=<-1 -1 0>
    Skeletal Mesh=$System.Skel
   )
   &CameraOffset(Camera Offset Amount=15)
   &ScaleColor(
    Curve Index=$System.LoopedAge
    Linear Color Curve=DataInterface() {propertyValues {
      AlphaCurve {keys [<at 0 1>,<at 0.4671 0.6>,<at 1 1>]}
      BlueCurve {keys [<at 0.1217 0.8714>,<at 0.3306 1>,<at 0.4729 0.7292>,<at 0.6336 0.8714>,<at 0.8205 0.8714>]}
      GreenCurve {keys [<at 0.1217 0.3515>,<at 0.3306 0.3241>,<at 0.4729 0.2521>,<at 0.6336 0.1894>,<at 0.8205 0.3515>]}
      RedCurve {keys [<at 0.1217 0.227>,<at 0.3306 0.3406>,<at 0.4729 0.1822>,<at 0.6336 0.2246>,<at 0.8205 0.227>]}
     }}
    Scale Mode=::NewEnumerator2
   )
   &FN_Light_v2(Light Exponent=2, Light Radius=100, LightEnabled=true, Scale Color=1)
   &ScaleColor(
    Scale Alpha=&FloatFromCurve(FloatCurve=DataInterface() {propertyValues {Curve {keys [
         {interpMode=RCIM_Cubic, tangentMode=RCTM_User, time=0, value=0}
         {arriveTangent=0.4296, interpMode=RCIM_Cubic, leaveTangent=0.4296, tangentMode=RCTM_User, time=0.3, value=1}
         {arriveTangent=-1.4286, interpMode=RCIM_Cubic, leaveTangent=-1.4286, time=0.8, value=1}
         {tangentMode=RCTM_User, time=1, value=0}
        ]}}})
    Scale RGB=&VectorFromFloat(Value=&FloatFromCurve(FloatCurve=DataInterface() {propertyValues {Curve {keys [
          {interpMode=RCIM_Cubic, tangentMode=RCTM_User, time=0, value=0}
          {interpMode=RCIM_Cubic, tangentMode=RCTM_User, time=0.2, value=1}
          {arriveTangent=-1.25, interpMode=RCIM_Cubic, leaveTangent=-1.25, time=0.8, value=1}
          {tangentMode=RCTM_User, time=1, value=0}
         ]}}}))
    ScaleA=true
    ScaleRGB=true
   ) {module=ScaleColor001}
  ]
  rendererClasses [&NiagaraSpriteRendererProperties, &NiagaraLightRendererProperties]
  renderers [
   {rendererClass=&NiagaraSpriteRendererProperties}
   {rendererClass=&NiagaraLightRendererProperties, rendererIndex=1}
  ]
 }}
summary {
 systemName=NS_AuraCop_Glow
 emitters [{emitterName=HandGlow, rendererClasses [&NiagaraSpriteRendererProperties, &NiagaraLightRendererProperties]}]
 userVariables [
  User.Flare {type=NiagaraFloat, defaultValue=0, description=""}
  User.FXBrightnessScale {type=NiagaraFloat, defaultValue=0.86, description=""}
  User.RampIn {type=NiagaraFloat, defaultValue=1, description=""}
  User.RampOut {type=NiagaraFloat, defaultValue=1, description=""}
 ]
}
